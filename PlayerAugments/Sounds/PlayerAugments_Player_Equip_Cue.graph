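# Sound cue for the player equip sound in PlayerAugments, judging by its name, using the augment selector's equip-device wave.
# A single wave player feeds the output directly, with no randomization or mixing.
# It uses the selector attenuation asset and the weapon foley sound class at slightly reduced volume.
# A concurrency override caps simultaneous instances, stopping the oldest, and the debug flag is left on.
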SoundCue /PlayerAugments/Sounds/PlayerAugments_Player_Equip_Cue
root=WavePlayer_1
nodes [{id=WavePlayer_1, properties {soundWaveAssetPtr=&/PlayerAugments_Selector_Equip_Device_01}}]
properties {
 attenuationSettings=&/PlayerAugments_Selector_Att
 bDebug=true
 bOverrideConcurrency=true
 maxDistance=3000
 soundClassObject=&Fort_WeaponFoley
 volumeMultiplier=0.752
 concurrencyOverrides {maxCount=3, resolutionRule=StopOldest, retriggerTime=0.5}
}
