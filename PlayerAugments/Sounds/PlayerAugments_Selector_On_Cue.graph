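# Sound cue for the player augments selector "On" event, classed as critical in-game UI audio.
# The node graph is empty with no root, so the cue references no sound waves as it stands.
# It carries only playback settings: a selector attenuation asset, disabled virtualization, and a concurrency override that caps simultaneous instances at two and stops the oldest.
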SoundCue /PlayerAugments/Sounds/PlayerAugments_Selector_On_Cue
root=null
nodes []
properties {
 attenuationSettings=&/PlayerAugments_Selector_Att
 bOverrideConcurrency=true
 maxDistance=3000
 soundClassObject=&Fort_UI_InGame_Critical
 virtualizationMode=Disabled
 volumeMultiplier=1.008
 concurrencyOverrides {maxCount=2, resolutionRule=StopOldest}
}
