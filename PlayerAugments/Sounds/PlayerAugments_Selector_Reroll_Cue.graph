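# Sound cue for the reroll action of the Player Augments selector, playing a single wave with no randomization or modulation.
# It routes to the Fort_Gadgets sound class with the selector's shared attenuation and a boosted volume.
# A concurrency override caps simultaneous instances at two, stopping the oldest, and virtualization is disabled.
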SoundCue /PlayerAugments/Sounds/PlayerAugments_Selector_Reroll_Cue
root=WavePlayer_0
nodes [{id=WavePlayer_0, properties {soundWaveAssetPtr=&/PlayerAugments_Selector_Reroll_01}}]
properties {
 attenuationSettings=&/PlayerAugments_Selector_Att
 bOverrideConcurrency=true
 maxDistance=3000
 soundClassObject=&Fort_Gadgets
 virtualizationMode=Disabled
 volumeMultiplier=1.5
 concurrencyOverrides {maxCount=2, resolutionRule=StopOldest}
}
