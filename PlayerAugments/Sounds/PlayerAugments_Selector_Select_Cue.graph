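# Sound cue that plays the select sound for the Player Augments selector.
# A single wave player feeds the cue directly, routed through the Fort_Gadgets sound class with boosted volume and a very large max distance.
# Its own concurrency override caps simultaneous plays at two, stops the oldest and applies a retrigger time, with virtualization disabled.
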SoundCue /PlayerAugments/Sounds/PlayerAugments_Selector_Select_Cue
root=WavePlayer_0
nodes [{id=WavePlayer_0, properties {soundWaveAssetPtr=&/PlayerAugments_Selector_Select_01}}]
properties {
 bOverrideConcurrency=true
 maxDistance=2097152
 soundClassObject=&Fort_Gadgets
 virtualizationMode=Disabled
 volumeMultiplier=1.3
 concurrencyOverrides {maxCount=2, resolutionRule=StopOldest, retriggerTime=1.5}
}
