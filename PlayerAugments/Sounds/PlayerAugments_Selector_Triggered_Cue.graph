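# Sound cue for the triggered sound of the player augment selector, routed to the critical in-game UI sound class.
# A single wave player feeds a modulator that lightly randomizes playback before output.
# The cue uses the selector's attenuation asset with a volume boost and virtualization disabled.
# An overridden concurrency limit keeps only a couple of instances playing, stopping the oldest first.
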SoundCue /PlayerAugments/Sounds/PlayerAugments_Selector_Triggered_Cue
root=Modulator_0
nodes [
 {id=Modulator_0, children [WavePlayer_1], properties {pitchMax=1}}
 {id=WavePlayer_1, properties {soundWaveAssetPtr=&/PlayerAugments_Triggered}}
]
properties {
 attenuationSettings=&/PlayerAugments_Selector_Att
 bOverrideConcurrency=true
 maxDistance=3000
 soundClassObject=&Fort_UI_InGame_Critical
 virtualizationMode=Disabled
 volumeMultiplier=1.3
 concurrencyOverrides {maxCount=2, resolutionRule=StopOldest}
}
