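# Niagara burst effect for a player selection on an augment granter, named for that moment, colored cyan-blue or purple by User.HordeRush.
# Core layers three animated sprites and a light flash, while Particle_Decal adds two staggered sprite bursts sized by House_Size_Scale.
# Sparks swirl upward under vortex, drag and positive-Z gravity.
# The medallion lightweight flag suppresses sparks and shifts offsets and timing; LightWeight_Value and time-of-day scalars tune brightness and opacity.
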NiagaraSystem /PlayerAugments/VFX/NS_Aug_Granter_Player_Select
dependencies {
 usedDataInterfaces [&NiagaraDataInterfaceCurve, &NiagaraDataInterfaceColorCurve, &NiagaraDataInterfaceVector2DCurve]
 usedDynamicInputs []
 usedModules [
  &self:SystemSpawnScript.SystemScriptSource.SystemScriptGraph.NiagaraNodeAssignment_1.SetVariables_#1
  &self:SystemSpawnScript.SystemScriptSource.SystemScriptGraph.NiagaraNodeAssignment_0.SetVariables_#2
  &SystemState
 ]
}
emitters {
 Core {
  emitterUpdateScript [
   &EmitterState()
   &SpawnBurst_Instantaneous(Spawn Count=1, Use Spawn Probability=false)
  ]
  particleSpawnScript [
   &InitializeParticle(
    Lifetime=0.7
    Position Offset=&MakeCustomVectorFromBool(Bool=$User.bls_Medallion_LightWeight, True Vector=<0 0 20>, False Vector=<0 0 0>)
    UsePositionOffset=true
    Write Material Random=true
   )
   /SetVariables(
    Particles.Sprite_Color_01:LinearColor=&MakeCustomLinearColorFromBool(Bool=$User.HordeRush, True Color=<rgba 1.3333 0 2 1>, False Color=<rgba 0 1.4263 2 1>)
    Particles.Sprite_DP_01:Vector4f=&MakeVector4(X=&RandomRangeFloat(Minimum=0.4, Maximum=0.6), Y=1, Z=1, W=1)
    Particles.Sprite_Size_01:Vector2f=&Vector2DFromFloat(Value=200)
   ) {isSetParameters=true}
   /SetVariables(
    Particles.Sprite_Color_02:LinearColor=&MakeCustomLinearColorFromBool(Bool=$User.HordeRush, True Color=<rgba 0.092 0 0.276 0.3>, False Color=<rgba 0 0.0654 0.276 0.3>)
    Particles.Sprite_DP_02:Vector4f=&MakeVector4(X=&RandomRangeFloat(Minimum=0.4, Maximum=0.6), Y=1, Z=1, W=1)
    Particles.Sprite_Size_02:Vector2f=&Vector2DFromFloat(Value=450)
   ) {isSetParameters=true}
   /SetVariables(
    Particles.Sprite_Color_03:LinearColor=&MakeCustomLinearColorFromBool(Bool=$User.HordeRush, True Color=<rgba 0.6667 0 1 3.5>, False Color=<rgba 0 0.5167 1 3.5>)
    Particles.Sprite_Size_03:Vector2f=&Vector2DFromFloat(Value=200)
   ) {isSetParameters=true}
  ]
  particleUpdateScript [
   &ParticleState()
   &CameraOffset(Camera Offset Amount=50)
   /SetVariables(
    Particles.Sprite_Color_01:LinearColor=&MultiplyLinearColor(
     Linear Color=$Particles.Initial.Sprite_Color_01
     Scale Factor=&MakeFloat4FromVectorAndFloat(
      "Vector (RGB)"=<1 1 1>
      "Float (Alpha)"=&Multiply_Float(
       A=&FNTimeOfDayScalar(Night Scale=0.5, Morning Scale=0.3, Evening Scale=0.4)
       B=&FloatFromCurve(
        FloatCurve=DataInterface() {propertyValues {Curve {keys [<at 0 0>,<at 0.2 1>,<at 0.8 1>,<at 1 0>]}}}
        CurveIndex=&Multiply_Float(A=$Particles.NormalizedAge, B=1.75)
        Scale Curve=$System.LightWeight_Value
       )
      )
     )
    )
    Particles.Sprite_Size_01:Vector2f=&Multiply_Vector2D(
     A=$Particles.Initial.Sprite_Size_01
     B=&Vector2DFromFloat(Value=&FloatFromCurve(
       FloatCurve=DataInterface() {propertyValues {Curve {keys [<at 0 1.5 RCIM_Cubic>,<at 1 0>]}}}
       CurveIndex=&Multiply_Float(A=$Particles.NormalizedAge, B=1.75)
      ))
    )
   ) {isSetParameters=true}
   /SetVariables(
    Particles.Sprite_Color_02:LinearColor=&MultiplyLinearColor(
     Linear Color=$Particles.Initial.Sprite_Color_02
     Scale Factor=&MakeFloat4FromVectorAndFloat(
      "Vector (RGB)"=&Multiply_VectorByFloat(Vector=<1 1 1>, Float=&FNTimeOfDayScalar(Night Scale=0.5, Morning Scale=0.3, Evening Scale=0.4))
      "Float (Alpha)"=&FloatFromCurve(
       FloatCurve=DataInterface() {propertyValues {Curve {keys [<at 0 0>,<at 0.2 1>,<at 0.8 1>,<at 1 0>]}}}
       CurveIndex=&Multiply_Float(A=$Particles.NormalizedAge, B=1.75)
       Scale Curve=$System.LightWeight_Value
      )
     )
    )
    Particles.Sprite_Size_02:Vector2f=&Multiply_Vector2D(
     A=$Particles.Initial.Sprite_Size_02
     B=&Vector2DFromFloat(Value=&FloatFromCurve(
       FloatCurve=DataInterface() {propertyValues {Curve {keys [<at 0 1.5 RCIM_Cubic>,<at 1 0>]}}}
       CurveIndex=&Multiply_Float(A=$Particles.NormalizedAge, B=1.75)
      ))
    )
   ) {isSetParameters=true}
   /SetVariables(
    Particles.Sprite_Color_03:LinearColor=&MultiplyLinearColor(
     Linear Color=$Particles.Initial.Sprite_Color_03
     Scale Factor=&MakeFloat4FromVectorAndFloat(
      "Vector (RGB)"=<1 1 1>
      "Float (Alpha)"=&Multiply_Float(
       A=&FloatFromCurve(
        FloatCurve=DataInterface() {propertyValues {Curve {keys [{interpMode=RCIM_Cubic, leaveTangent=-0.6434, tangentMode=RCTM_Break, time=0, value=1}, <at 1 0>]}}}
        CurveIndex=&Multiply_Float(A=$Particles.NormalizedAge, B=2)
        Scale Curve=$System.LightWeight_Value
       )
       B=&FNTimeOfDayScalar(Night Scale=1, Morning Scale=0.3, Evening Scale=0.5)
      )
     )
    )
    Particles.Sprite_Size_03:Vector2f=&Multiply_Vector2D(
     A=$Particles.Initial.Sprite_Size_03
     B=&Vector2DFromFloat(Value=&FloatFromCurve(
       FloatCurve=DataInterface() {propertyValues {Curve {keys [{tangentMode=RCTM_User, time=0, value=0}, <at 1 1>]}}}
       CurveIndex=&Multiply_Float(A=$Particles.NormalizedAge, B=2)
      ))
    )
   ) {isSetParameters=true}
   &FN_Light_Attributes(
    Color=&MultiplyLinearColor(
     Linear Color=&MakeCustomLinearColorFromBool(Bool=$User.HordeRush, True Color=<rgba 2669.2708 338.541 5000 1>, False Color=<rgba 338.541 1814.6604 5000 1>)
     Scale Factor=&MakeVector4_FromLinearColor(LinearColor=&ColorFromCurve(ColorCurve=DataInterface() {propertyValues {
         AlphaCurve {keys [<at 0 0>,<at 0.2565 1>,<at 1 0>]}
         BlueCurve {keys [<at 0 1>,<at 0.2541 5>,<at 0.303 1>,<at 1 1>]}
         GreenCurve {keys [<at 0 1>,<at 0.2541 5>,<at 0.303 1>,<at 1 1>]}
         RedCurve {keys [<at 0 1>,<at 0.2541 5>,<at 0.303 1>,<at 1 1>]}
        }}))
    )
    Create Color Attribute=true
    Scale Color=$System.LightWeight_Value
   )
  ]
  rendererClasses [&NiagaraLightRendererProperties, &NiagaraSpriteRendererProperties]
  renderers [
   {rendererClass=&NiagaraLightRendererProperties, rendererIndex=0}
   {rendererClass=&NiagaraSpriteRendererProperties, rendererIndex=1}
   {rendererClass=&NiagaraSpriteRendererProperties, rendererIndex=2}
   {rendererClass=&NiagaraSpriteRendererProperties, rendererIndex=3}
  ]
 }
 Particle_Decal {
  emitterUpdateScript [
   &EmitterState()
   &SpawnBurst_Instantaneous(Spawn Count=1, Use Spawn Probability=false)
   &SpawnBurst_Instantaneous(
    Spawn Time=&MakeCustomFloatFromBool(Bool=$User.bls_Medallion_LightWeight, True Float=0.05, False Float=0.075)
    Use Spawn Probability=false
    UseLoopCountLimit=false
   ) {module=SpawnBurst_Instantaneous001}
  ]
  particleSpawnScript [
   &InitializeParticle(
    Color=&MakeCustomLinearColorFromBool(Bool=$User.HordeRush, True Color=<rgba 3.3333 0 5 1>, False Color=<rgba 0 5.4465 13 1>)
    Sprite Size Mode=::NewEnumerator3
    Uniform Sprite Size=185
    UsePositionOffset=false
    Write Material Random=true
   )
   /SetVariables(Particles.Sprite02_CameraOffset:NiagaraFloat=&Multiply_Float(A=40, B=1)) {isSetParameters=true}
   /SetVariables(Particles.Sprite03_CameraOffset:NiagaraFloat=&Multiply_Float(A=15, B=1)) {isSetParameters=true}
   &SystemLocation(MaskSpawnGroup=false, Offset=<0 0 -25>)
  ]
  particleUpdateScript [
   &CameraOffset(Camera Offset Amount=&Multiply_Float(A=50, B=1))
   &ParticleState()
   &ScaleColor(
    Scale Alpha=&FloatFromCurve(FloatCurve=DataInterface() {propertyValues {Curve {
        defaultValue=1
        keys [
         {arriveTangent=9.0725, interpMode=RCIM_Cubic, leaveTangent=9.0725, tangentMode=RCTM_User, time=0, value=0}
         {arriveTangent=-2.5249, interpMode=RCIM_Cubic, leaveTangent=-2.5249, tangentMode=RCTM_User, time=0.1408, value=0.9847}
         {arriveTangent=-0.0405, interpMode=RCIM_Cubic, leaveTangent=-0.0405, tangentMode=RCTM_User, time=1, value=0}
        ]
       }}})
    Scale RGB=/Niagara/DynamicInputs/TypeConversions/MakeVectorFromLinearColorHSV(Color:LinearColor=&MakeCustomLinearColorFromBool(
      Bool=$User.HordeRush
      True Color=&ColorFromCurve(ColorCurve=DataInterface() {propertyValues {
         AlphaCurve {keys [<at 0 0 RCIM_Cubic>, {arriveTangent=-1.373, interpMode=RCIM_Cubic, leaveTangent=-1.373, tangentMode=RCTM_User, time=0.3, value=1}, <at 1 0>]}
         BlueCurve {keys [<at 0 1>,<at 0.6831 1>]}
         GreenCurve {keys [<at 0 0>,<at 0.6831 0>]}
         RedCurve {keys [<at 0 0.8333>,<at 0.6831 0.3333>]}
        }})
      False Color=&ColorFromCurve(ColorCurve=DataInterface() {propertyValues {
         AlphaCurve {keys [<at 0 0 RCIM_Cubic>, {arriveTangent=-1.373, interpMode=RCIM_Cubic, leaveTangent=-1.373, tangentMode=RCTM_User, time=0.3, value=1}, <at 1 0>]}
         BlueCurve {keys [<at 0 1>,<at 0.6831 1>]}
         GreenCurve {keys [<at 0 0.9225>,<at 0.6831 0.5727>]}
         RedCurve {keys [<at 0 0>,<at 0.6831 0>]}
        }})
     ))
    ScaleA=true
    ScaleRGB=true
   )
   &DynamicMaterialParameters(
    Index 0 Param 1=&FloatFromCurve(FloatCurve=DataInterface() {propertyValues {Curve {keys [
         <at 0 -0.5 RCIM_Cubic>
         {arriveTangent=21.4883, interpMode=RCIM_Cubic, leaveTangent=21.4883, tangentMode=RCTM_User, time=0.07, value=0.27}
         {arriveTangent=5.7766, interpMode=RCIM_Cubic, leaveTangent=5.7766, tangentMode=RCTM_User, time=0.15, value=1}
         {arriveTangent=0.5047, interpMode=RCIM_Cubic, leaveTangent=0.5047, tangentMode=RCTM_User, time=1, value=5}
        ]}}})
    Index 0 Param 2=1
    Index 0 Param 3=&FloatFromCurve(FloatCurve=DataInterface() {propertyValues {Curve {keys [{arriveTangent=1.3198, interpMode=RCIM_Cubic, leaveTangent=1.3198, tangentMode=RCTM_User, time=0, value=0}, <at 1 1>]}}})
    Index 0 Param 4=1
    Module.Param0WriteEnabled=true
    Module.Param1WriteEnabled=true
    Module.Param2WriteEnabled=true
    Module.Param3WriteEnabled=true
   )
   &ScaleColor(
    Scale Alpha=&Multiply_Float(
     A=&FNTimeOfDayScalar(Night Scale=0.7, Morning Scale=0.2, Evening Scale=0.2)
     B=$System.LightWeight_Value
    )
    ScaleA=true
    ScaleRGB=true
   ) {module=ScaleColor001}
   &ScaleSpriteSize(Scale Factor=&Vector2DFromFloat(Value=$User.House_Size_Scale))
  ]
  rendererClasses [&NiagaraSpriteRendererProperties]
  renderers [
   {rendererClass=&NiagaraSpriteRendererProperties}
   {rendererClass=&NiagaraSpriteRendererProperties, rendererIndex=1}
   {rendererClass=&NiagaraSpriteRendererProperties, rendererIndex=2}
   {rendererClass=&NiagaraSpriteRendererProperties, rendererIndex=3}
   {rendererClass=&NiagaraSpriteRendererProperties, rendererIndex=4}
   {rendererClass=&NiagaraSpriteRendererProperties, rendererIndex=5}
  ]
 }
 Sparks {
  emitterUpdateScript [
   &EmitterState()
   &SpawnBurst_Instantaneous(
    Spawn Count=25
    Spawn Probability=&MakeCustomFloatFromBool(Bool=$User.bls_Medallion_LightWeight, True Float=0, False Float=1)
    Use Spawn Probability=true
    UseLoopCountLimit=false
   )
  ]
  particleSpawnScript [
   &InitializeParticle(
    Color=&MakeCustomLinearColorFromBool(
     Bool=$User.HordeRush
     True Color=&MakeCustomLinearColorFromBool(Bool=&RandomBool(), True Color=<rgba 1.6667 0 5 1>, False Color=<rgba 2 0 2 1>)
     False Color=&MakeCustomLinearColorFromBool(Bool=&RandomBool(), True Color=<rgba 0 4.0158 5 1>, False Color=<rgba 0 0.2176 2 1>)
    )
    Lifetime Max=1
    Lifetime Min=0.25
    Lifetime Mode=::NewEnumerator1
    Position Mode=::NewEnumerator1
    Sprite Size Mode=::NewEnumerator3
    Uniform Sprite Size=&Lerp_Float(
     A=3
     B=7
     Alpha=&FloatFromCurve(
      FloatCurve=DataInterface() {propertyValues {Curve {keys [
          {interpMode=RCIM_Cubic, tangentMode=RCTM_User, time=0, value=0}
          {arriveTangent=1.2457, interpMode=RCIM_Cubic, leaveTangent=1.2457, tangentMode=RCTM_User, time=0.7348, value=0.4334}
          {interpMode=RCIM_Cubic, tangentMode=RCTM_User, time=1, value=1}
         ]}}}
      CurveIndex=&RandomRangeFloat(Minimum=0, Maximum=1)
     )
    )
    UsePositionOffset=false
    Write Material Random=true
   )
   &ShapeLocation(MaskSpawnGroup=false, Offset Mode=::NewEnumerator0, Override Seed=false, Sphere Radius=3, UseOwnerScale=true)
   &AddVelocity(
    Module.Use Falloff Distance=false
    Override Seed=false
    Velocity Mode=::NewEnumerator1
    Velocity Speed=&RandomRangeFloat(Minimum=104.8031, Maximum=188.6381)
   )
   /SetVariables(Particles.Sprite02_Size:Vector2f=&Vector2DFromFloat(Value=&Multiply_Float(A=13, B=1))) {isSetParameters=true}
   &SystemLocation(
    MaskSpawnGroup=false
    Offset=&MakeCustomVectorFromBool(Bool=$User.bls_Medallion_LightWeight, True Vector=<0 0 20>, False Vector=<0 0 0>)
   )
  ]
  particleUpdateScript [
   &GravityForce(Gravity=&RandomRangeVector(Minimum=<0 0 66.792>, Maximum=<0 0 125.704>))
   &Drag(Drag=2, Use Linear Drag=true, Use Rotational Drag=true)
   &VortexForce(UseOriginOffset=true, Vortex Force Amount=339.4481, Vortex Origin:NiagaraPosition=$Particles.Initial.ShapeLocation.ShapePosition)
   &SolveForcesAndVelocity(Clamp Velocity=false, Limit Acceleration=false)
   /SetVariables(Particles.Sprite02_Color:LinearColor=&MakeLinearColorFromVectorAndFloat(
     "Vector (RGB)"=&MakeVectorFromLinearColorRGB(Color=&MakeCustomLinearColorFromBool(Bool=$User.HordeRush, True Color=<rgba 0.6832 0.0496 1 1>, False Color=<rgba 0.0496 0.5658 1 1>))
     "Float (Alpha)"=&FloatFromCurve(FloatCurve=DataInterface() {propertyValues {Curve {keys [
          <at 0 1 RCIM_Cubic>
          {arriveTangentWeight=0.2357, interpMode=RCIM_Cubic, leaveTangentWeight=0.2357, tangentMode=RCTM_User, time=0.5, value=1}
          <at 1 0 RCIM_Cubic>
         ]}}})
    )) {isSetParameters=true}
   &ParticleState()
   &ScaleSpriteSize(Uniform Curve Sprite Scale=DataInterface() {propertyValues {Curve {keys [
        {arriveTangent=8.2127, interpMode=RCIM_Cubic, leaveTangent=8.2127, tangentMode=RCTM_User, time=0, value=0.5}
        {
         arriveTangent=4.2448
         arriveTangentWeight=0.232
         interpMode=RCIM_Cubic
         leaveTangent=4.2448
         leaveTangentWeight=2.2538
         tangentMode=RCTM_User
         time=0.07
         value=1
        }
        {arriveTangent=-2.5194, interpMode=RCIM_Cubic, leaveTangent=-2.5194, tangentMode=RCTM_User, time=0.75, value=0.8}
        {arriveTangent=-4.4982, interpMode=RCIM_Cubic, leaveTangent=-4.4982, tangentMode=RCTM_User, time=1, value=0}
       ]}}})
   &ScaleColor(
    Scale Alpha=&FloatFromCurve(
     FloatCurve=DataInterface() {propertyValues {Curve {keys [
         <at 0 1 RCIM_Cubic>
         {arriveTangentWeight=0.2357, interpMode=RCIM_Cubic, leaveTangentWeight=0.2357, tangentMode=RCTM_User, time=0.5, value=1}
         <at 1 0 RCIM_Cubic>
        ]}}}
     Scale Curve=$System.LightWeight_Value
    )
    ScaleA=true
    ScaleRGB=true
   )
   &DynamicMaterialParameters(
    Index 0 Param 1=&FloatFromCurve(
     FloatCurve=DataInterface() {propertyValues {Curve {keys [
         {arriveTangent=-5.8452, interpMode=RCIM_Cubic, time=0, value=1}
         {arriveTangentWeight=0.0434, interpMode=RCIM_Cubic, leaveTangentWeight=0.217, tangentMode=RCTM_User, time=0.2, value=0}
         {arriveTangent=3.4164, interpMode=RCIM_Cubic, leaveTangent=3.4164, tangentMode=RCTM_User, time=0.8, value=0.7}
         <at 1 0 RCIM_Cubic>
        ]}}}
     Scale Curve=3
    )
    Index 0 Param 2=10
    Index 0 Param 3=700
    Index 0 Param 4=5
    Module.Param0WriteEnabled=true
    Module.Param1WriteEnabled=true
    Module.Param2WriteEnabled=true
    Module.Param3WriteEnabled=true
   )
  ]
  rendererClasses [&NiagaraSpriteRendererProperties]
  renderers [
   {rendererClass=&NiagaraSpriteRendererProperties}
   {rendererClass=&NiagaraSpriteRendererProperties, rendererIndex=1}
  ]
 }
}
summary {
 systemName=NS_Aug_Granter_Player_Select
 emitters [
  {emitterName=Core, rendererClasses [&NiagaraLightRendererProperties, &NiagaraSpriteRendererProperties]}
  {emitterName=Sparks, rendererClasses [&NiagaraSpriteRendererProperties]}
  {emitterName=Particle_Decal, rendererClasses [&NiagaraSpriteRendererProperties]}
 ]
 userVariables [
  User.bIsOwningClient {type=NiagaraBool}
  User.bls_Medallion_LightWeight {type=NiagaraBool}
  User.HordeRush {type=NiagaraBool, defaultValue=true}
  User.House_Size_Scale {type=NiagaraFloat, defaultValue=1, description=""}
  User.OwningClient {type=NiagaraBool}
 ]
}
